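; Season 33 Story Confrontation quest altar that changes which splinter and Splinter Blade meshes are visible as quest side effects arrive.
; BP_ApplySideEffect switches on SplinterBladeAltar gameplay tags, passing stages S0–S5 as booleans to UpdateMeshVisibility or toggling the Niagara effect.
; Each stage from S0 to S4 sets only one or two meshes hidden or shown, while S5 changes nothing.
; The Niagara effect starts hidden on BeginPlay until the EnableVFX tag arrives.
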(event EventBeginPlay
  (|CallParentFunction)
  (CallFunction|DisableVFXOnStartup))

(event Collision|EventActorBeginOverlap
  (params (OtherActor :type Actor%20Object%20Reference)))

(event EventTick
  (params (DeltaSeconds :type Float%20%28single-precision%29)))

(event SideEffects|EventBP_ApplySideEffect
  (params (ReactionSource :type Object%20Reference) (TargetObject :type Object%20Reference) (SideEffectTriggerTag :type Gameplay%20Tag%20Structure%20%28by%20ref%29) (SideEffectTriggerType :type EBlueprintableSideEffectTriggerType%20Enum%20%28by%20ref%29 :default "TriggeredAsTarget"))
  (switch Utilities|FlowControl|Switch|SwitchonGameplayTag SideEffectTriggerTag
    (:Athena.Quests.S33.Story.Confrontation.SplinterBladeAltar.S0
      (CallFunction|UpdateMeshVisibility true))
    (:Athena.Quests.S33.Story.Confrontation.SplinterBladeAltar.S1
      (CallFunction|UpdateMeshVisibility false true))
    (:Athena.Quests.S33.Story.Confrontation.SplinterBladeAltar.S2
      (CallFunction|UpdateMeshVisibility false false true))
    (:Athena.Quests.S33.Story.Confrontation.SplinterBladeAltar.S3
      (CallFunction|UpdateMeshVisibility false false false true))
    (:Athena.Quests.S33.Story.Confrontation.SplinterBladeAltar.S4
      (CallFunction|UpdateMeshVisibility false false false false true))
    (:Athena.Quests.S33.Story.Confrontation.SplinterBladeAltar.S5
      (CallFunction|UpdateMeshVisibility false false false false false true))
    (:Athena.Quests.S33.Story.Confrontation.SplinterBladeAltar.DisableVFX
      (Development|SetHiddeninGame (Variables|Default|GetNiagara) true false))
    (:Athena.Quests.S33.Story.Confrontation.SplinterBladeAltar.EnableVFX
      (Development|SetHiddeninGame (Variables|Default|GetNiagara) false false))))

(event Custom|DisableVFXOnStartup
  (Development|SetHiddeninGame (Variables|Default|GetNiagara) true false))

(fn UpdateMeshVisibility
  (params (S0 :type Boolean) (S1 :type Boolean) (S2 :type Boolean) (S3 :type Boolean) (S4 :type Boolean) (S5 :type Boolean))
  (if S0
    (Development|SetHiddeninGame (Variables|Default|GetSplinter1) true false)
    (return)
    (elif S1
      (Development|SetHiddeninGame (Variables|Default|GetSplinter2) true false)
      (Development|SetHiddeninGame (Variables|Default|GetSplinter1) false false)
      (return)
      (elif S2
        (Development|SetHiddeninGame (Variables|Default|GetSplinter3) true false)
        (Development|SetHiddeninGame (Variables|Default|GetSplinter1) false false)
        (return)
        (elif S3
          (Development|SetHiddeninGame (Variables|Default|GetSplinterBlade1) true false)
          (Development|SetHiddeninGame (Variables|Default|GetSplinter1) false false)
          (return)
          (elif S4
            (Development|SetHiddeninGame (Variables|Default|GetSplinter3) true false)
            (Development|SetHiddeninGame (Variables|Default|GetSplinterBlade1) false false)
            (return)
            (elif S5
              (return)
              (else
                (return)))))))))

(fn ConstructionScript)
