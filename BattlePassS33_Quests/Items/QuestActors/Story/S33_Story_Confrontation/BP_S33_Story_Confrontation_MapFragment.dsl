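; Map fragment quest actor for the Season 33 story Confrontation quest, shown as a spinning, bobbing mesh until a successful overlap.
; On BeginPlay the SMRotator timeline starts, yawing the tilted static mesh through a full turn and bobbing its relative height around an offset derived from the capsule.
; The OverlapSuccessful override calls the parent, then hides the mesh and its minimap quest icon.
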(event EventBeginPlay
  (|CallParentFunction)
  (CallFunction|StartSMRotator))

(event EventTick
  (params (DeltaSeconds :type Float%20%28single-precision%29)))

(event Custom|StartSMRotator
  (bind (_direction _rotation _bob) (Timeline|SMRotator PlayFromStart)
    (:Update
      (Transformation|SetWorldRotation (Variables|Default|GetStaticMesh) (Math|Rotator|MakeRotator 0.0 6e+01 (Math|Float|Lerp 0.0 3.6e+02 _rotation)) false false)
      (Transformation|SetRelativeLocation (Variables|Default|GetStaticMesh) (Math|Vector|MakeVector 0.0 0.0 (+ (+ (- (.z (Transformation|GetWorldLocation (Variables|Default|GetCapsule))) (.z (Transformation|GetActorLocation))) (Math|Float|Lerp 5.0 -5.0 _bob)) 32.0)) false false))))

(fn OverlapSuccessful
  (|CallParentFunction)
  (Rendering|SetVisibility (Variables|Default|GetStaticMesh) false false)
  (MiniMapData|SetVisibility (Variables|Default|GetFortQuestIcon) false))

(fn ConstructionScript
  (|CallParentFunction))
