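; Splinter loot actor for the Season 33 Story Confrontation quest that spins its static mesh.
; BeginPlay calls the parent, then StartSMRotator plays a timeline from the start, lerping its rotation track to a full 360-degree yaw applied as world rotation.
; Overlap and Tick events are empty, so any pickup behavior is not defined in this graph.
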(event EventBeginPlay
  (|CallParentFunction)
  (CallFunction|StartSMRotator))

(event Collision|EventActorBeginOverlap
  (params (OtherActor :type Actor%20Object%20Reference)))

(event EventTick
  (params (DeltaSeconds :type Float%20%28single-precision%29)))

(event Custom|StartSMRotator
  (bind (_direction _rotation) (Timeline|Timeline PlayFromStart)
    (:Update
      (Transformation|SetWorldRotation (Variables|Default|GetStaticMesh) (Math|Rotator|MakeRotator 0.0 0.0 (Math|Float|Lerp 0.0 3.6e+02 _rotation)) false false))))

(fn ConstructionScript)
